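# Translucent unlit material for Niagara sprites in the SugarBubble pickaxe effects, apparently drawing an emoji or graffiti symbol.
# It samples the T_Symbol_Graffiti texture and runs its RGB through the shared MF_FX_EmissiveAdjustments function for emissive.
# The texture's alpha is multiplied by particle color alpha, so particles control the symbol's fade.
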Material /BRCosmetics/Weapons/FORT_Melee/Pickaxe_SugarBubble/FX/M_Pickaxe_SugarBubble_Emoji
graph {
 properties {bUsedWithNiagaraSprites=true, blendMode=BLEND_Translucent, pixelDepthOffsetMode=PDOM_Legacy, shadingModel=MSM_Unlit, usedShadingModels=Unlit}
 nodes [
  n1 = TextureSample() {texture=&/T_Symbol_Graffiti}
  MaterialFunctionCall(Emissive=@n1.RGB) {materialFunction=&MF_FX_EmissiveAdjustments}
  Multiply(A=@n1.A, B=ParticleColor().A) {constA=1}
 ]
}
