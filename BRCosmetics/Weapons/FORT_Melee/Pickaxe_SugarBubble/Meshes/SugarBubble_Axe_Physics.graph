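# Physics asset for the SugarBubble two-handed pickaxe mesh, setting up a dangling charm that can swing from the handle.
# A capsule body covers the handle, and five smaller capsules cover the charm bones.
# Angular-limited constraints link the charm bodies in a serial chain from the handle.
# The first joint is tight on twist and one swing axis, while the joints further down allow wider swing on both axes.
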PhysicsAsset /BRCosmetics/Weapons/FORT_Melee/Pickaxe_SugarBubble/Meshes/SugarBubble_Axe_Physics
bodies {
 Melee_TwoHanded_Handle [{shapeType=Capsule, length=42.281, radius=2.6327, center=<0.5935 0.2738 5.9283>, rotation=<rot 0.478 -90.1594 -0.774>}]
 charm_1_dyn [{
   shapeType=Capsule
   length=2.0534
   radius=1.0405
   shapeName=charm_1_dyn_capsule
   center=<1.443 0.0254 -0.0144>
   rotation=<rot 0 -92.4509 -87.6077>
  }]
 charm_2_dyn [{shapeType=Capsule, length=3.1324, radius=0.596, shapeName=charm_2_dyn_capsule, rotation=<rot 3.5884 -90 -90.489>}]
 charm_3_dyn [{shapeType=Capsule, length=3.8275, radius=0.6978, shapeName=charm_3_dyn_capsule, rotation=<rot 0.0004 -92.7395 -91.4815>}]
 charm_4_dyn [{
   shapeType=Capsule
   length=2.5784
   radius=1.6168
   shapeName=charm_4_dyn_capsule
   center=<1.5959 0.1178 0.0234>
   rotation=<rot 0 -92.0705 -89.3686>
  }]
 charm_5_dyn [{shapeType=Capsule, length=3.7891, radius=3.5592, shapeName=charm_5_dyn_capsule, rotation=<rot 0.0002 -95.4659 -90.4178>}]
}
constraints [
 {bone1Name=charm_1_dyn, bone2Name=Melee_TwoHanded_Handle, swing1LimitDegrees=2, swing2LimitDegrees=70, twistLimitDegrees=2}
 {bone1Name=charm_2_dyn, bone2Name=charm_1_dyn, swing1LimitDegrees=45, swing2LimitDegrees=50, twistLimitDegrees=7}
 {bone1Name=charm_3_dyn, bone2Name=charm_2_dyn, swing1LimitDegrees=45, swing2LimitDegrees=50, twistLimitDegrees=7}
 {bone1Name=charm_4_dyn, bone2Name=charm_3_dyn, swing1LimitDegrees=45, swing2LimitDegrees=50, twistLimitDegrees=6}
 {bone1Name=charm_5_dyn, bone2Name=charm_4_dyn, swing1LimitDegrees=75, swing2LimitDegrees=45, twistLimitDegrees=11}
]
